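; Creative Timer Objective device: a countdown that players or incoming events can activate, pause, deactivate, restart or complete.
; Minigame changes, team-filtered interactions and custom events feed SetState.
; OnRep_State applies each state: timers, server-side gameplay event broadcasts, messages, audio and VFX.
; A repeating clock timer refreshes displays and minimap time, while options drive visibility, hologram materials and an alternate explosive-device mesh.
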(event Interaction|EventBlueprintOnBeginInteract
  (params (InteractingPawn :type Fort%20Pawn%20Object%20Reference) (InteractionBeingAttempted :type EInteractionBeingAttempted%20Enum :default "FirstInteraction") (InteractionType :type TInteractionType%20Enum :default "IT_NoInteraction"))
  ;[0 0 1] Handle Begin Interaction
  (if (Math|Conversions|ToBoolean(Integer) (Variables|ToyProperties|GetSounds))
    (switch Utilities|FlowControl|Switch|SwitchonE_TimerObjective_State (Variables|Transient|GetState)
      (:NewEnumerator7
        _)
      (:NewEnumerator1
        _)
      (:NewEnumerator3
        (CallFunction|PlaySound (Variables|Audio|GetStartDefuseInteractionSound))
        (if (not (Audio|Components|Audio|IsPlaying (Variables|Default|GetDefuse%20Power%20Down%20Audio%20Loop)))
          (Audio|Components|Audio|FadeIn (Variables|Default|GetDefuse%20Power%20Down%20Audio%20Loop) 1.0 1.0 0.0 "Linear" :!class /Script/Engine.AudioComponent)))
      (:NewEnumerator10
        _))
    (CallFunction|PlaySound (Variables|Audio|GetStartPlantingInteractionSound))
    (if (not (Audio|Components|Audio|IsPlaying (Variables|Default|GetPlant%20Power%20Up%20Audio%20Loop)))
      (Audio|Components|Audio|FadeIn (Variables|Default|GetPlant%20Power%20Up%20Audio%20Loop) 1.0 1.0 0.0 "Linear" :!class /Script/Engine.AudioComponent))))

(event Interaction|EventBlueprintOnInterruptInteract
  ;[0 1 0] Handle Interrupt Interact
  (if (Math|Conversions|ToBoolean(Integer) (Variables|ToyProperties|GetSounds))
    (Utilities|FlowControl|Sequence
      (:then_0
        (if (Audio|Components|Audio|IsPlaying (Variables|Default|GetPlant%20Power%20Up%20Audio%20Loop))
          (Audio|Components|Audio|FadeOut (Variables|Default|GetPlant%20Power%20Up%20Audio%20Loop) 1.0 0.0 "Linear" :!class /Script/Engine.AudioComponent)))
      (:then_1
        (if (Audio|Components|Audio|IsPlaying (Variables|Default|GetDefuse%20Power%20Down%20Audio%20Loop))
          (Audio|Components|Audio|FadeOut (Variables|Default|GetDefuse%20Power%20Down%20Audio%20Loop) 1.0 0.0 "Linear" :!class /Script/Engine.AudioComponent))))))

(event OnAnyPropertyChanged(ToyOptionsComponent)
  (Class|CreativeUIBroadcastComponent|Initialize (Variables|Default|GetCreative_HUDBroadcast_Component) (Variables|Default|GetFortMinigameProgress) (Variables|UserOptions|Advanced|GetDisplayScoreUpdateOnHUD))
  (CallFunction|RefreshMeshVisibilities)
  (Networking|CanExecuteCosmeticEvents
    (:True
      (CallFunction|UpdateVisualDisplays)
      (CallFunction|UpdateAdditionalMeshes)
      (Rendering|Material|SetCustomPrimitiveDataVector3 (Variables|Debug|GetStaticMeshComponent) (Variables|Material|GetEmissiveCustomPrimitiveIndex) (Math|Conversions|ToVector(LinearColor) (Variables|Colors|GetEmissiveIdleColor))))
    (:False
      (CallFunction|UpdateMinimapTime)
      (Class|CreativeMapMarkerComponent|SetVisibilityonMap (Variables|Default|GetCreativeMapMarkerComponent) (Variables|ToyProperties|GetShowOnMaps)))))

(event OnPropertiesLoaded(ToyOptionsComponent)
  (Class|CreativeUIBroadcastComponent|Initialize (Variables|Default|GetCreative_HUDBroadcast_Component) (Variables|Default|GetFortMinigameProgress) (Variables|UserOptions|Advanced|GetDisplayScoreUpdateOnHUD))
  (CallFunction|RefreshMeshVisibilities)
  (Networking|CanExecuteCosmeticEvents
    (:True
      (CallFunction|UpdateVisualDisplays)
      (CallFunction|UpdateAdditionalMeshes)
      (Rendering|Material|SetCustomPrimitiveDataVector3 (Variables|Debug|GetStaticMeshComponent) (Variables|Material|GetEmissiveCustomPrimitiveIndex) (Math|Conversions|ToVector(LinearColor) (Variables|Colors|GetEmissiveIdleColor))))
    (:False
      (CallFunction|UpdateMinimapTime)
      (Class|CreativeMapMarkerComponent|SetVisibilityonMap (Variables|Default|GetCreativeMapMarkerComponent) (Variables|ToyProperties|GetShowOnMaps)))))

(event EventBeginPlay
  (Minigame|BindEventtoMinigameStateChangedDelegate (Variables|Default|GetFortMinigameLogic) (EventDispatchers|CreateEvent :!event OnMinigameStateChanged))
  (CallFunction|RefreshMeshVisibilities)
  (Networking|CanExecuteCosmeticEvents
    (:True
      (CallFunction|UpdateVisualDisplays)
      (CallFunction|UpdateAdditionalMeshes)
      (Rendering|Material|SetCustomPrimitiveDataVector3 (Variables|Debug|GetStaticMeshComponent) (Variables|Material|GetEmissiveCustomPrimitiveIndex) (Math|Conversions|ToVector(LinearColor) (Variables|Colors|GetEmissiveIdleColor))))
    (:False
      (CallFunction|UpdateMinimapTime)
      (Class|CreativeMapMarkerComponent|SetVisibilityonMap (Variables|Default|GetCreativeMapMarkerComponent) (Variables|ToyProperties|GetShowOnMaps)))))

(event OnVisibilityChanged(StaticMeshVisibleInGameComponent)
  (params (Visible :type Boolean))
  (CallFunction|UpdateMeshesCollisionProfiles)
  (Networking|CanExecuteCosmeticEvents
    (:True
      (CallFunction|RefreshFoliage Visible)
      (CallFunction|UpdateAdditionalMeshes)
      (CallFunction|UpdateVisualDisplays))))

(event EventOnWorldReady
  (|CallParentFunction)
  (CallFunction|UpdateVisibilityComponents))

(event Interaction|EventBlueprintOnInteract
  (params (InteractingPawn :type Fort%20Pawn%20Object%20Reference) (InteractionBeingAttempted :type EInteractionBeingAttempted%20Enum :default "FirstInteraction"))
  (CallFunction|HandleonInteract InteractingPawn))

(event Custom|Activated
  (params (Instigator :type Controller%20Object%20Reference))
  ;[0.4909 0 1 0.25] Gameplay Functions
  (switch Utilities|FlowControl|Switch|SwitchonE_TimerObjective_State (Variables|Transient|GetState)
    (:NewEnumerator7
      _)
    (:NewEnumerator1
      _)
    (:NewEnumerator4
      _))
  (CallFunction|SetState "NewEnumerator3" Instigator))

(event Custom|Deactivated
  (params (Instigator :type Controller%20Object%20Reference))
  ;[0.4909 0 1 0.25] Gameplay Functions
  (switch Utilities|FlowControl|Switch|SwitchonE_TimerObjective_State (Variables|Transient|GetState)
    (:NewEnumerator3
      _)
    (:NewEnumerator4
      _))
  (CallFunction|SetState "NewEnumerator1" Instigator))

(event Custom|Paused
  (params (Instigator :type Controller%20Object%20Reference))
  ;[0.4909 0 1 0.25] Gameplay Functions
  (if (CallFunction|CanPause)
    (switch Utilities|FlowControl|Switch|SwitchonE_TimerObjective_State (Variables|Transient|GetState)
      (:NewEnumerator3
        (CallFunction|SetState "NewEnumerator4" Instigator)))))

(event Custom|Unpaused
  (params (Instigator :type Controller%20Object%20Reference))
  ;[0.4909 0 1 0.25] Gameplay Functions
  (if (CallFunction|CanPause)
    (switch Utilities|FlowControl|Switch|SwitchonE_TimerObjective_State (Variables|Transient|GetState)
      (:NewEnumerator4
        (CallFunction|SetState "NewEnumerator3" Instigator)))))

(event Custom|Complete%20Event
  (params (Instigator :type Controller%20Object%20Reference))
  ;[0.4909 0 1 0.25] Gameplay Functions
  (switch Utilities|FlowControl|Switch|SwitchonE_TimerObjective_State (Variables|Transient|GetState)
    (:NewEnumerator7
      _)
    (:NewEnumerator1
      _)
    (:NewEnumerator3
      _)
    (:NewEnumerator4
      _))
  (CallFunction|SetState "NewEnumerator5" Instigator))

(event Custom|Enabled
  (params (Instigator :type Controller%20Object%20Reference))
  ;[0.4909 0 1 0.25] Gameplay Functions
  (switch Utilities|FlowControl|Switch|SwitchonE_TimerObjective_State (Variables|Transient|GetState)
    (:NewEnumerator5
      _)
    (:NewEnumerator10
      _))
  (CallFunction|SetState "NewEnumerator7" Instigator))

(event Custom|Disabled
  (params (Instigator :type Controller%20Object%20Reference))
  ;[0.4909 0 1 0.25] Gameplay Functions
  (switch Utilities|FlowControl|Switch|SwitchonE_TimerObjective_State (Variables|Transient|GetState)
    (:NewEnumerator7
      _)
    (:NewEnumerator1
      _)
    (:NewEnumerator3
      _)
    (:NewEnumerator4
      _))
  (CallFunction|SetState "NewEnumerator10" Instigator))

(event Custom|ShowWhenReceivingFrom
  (params (Instigator :type Controller%20Object%20Reference))
  ;[0.4909 0 1 0.25] Gameplay Functions
  (CallFunction|SetForcedVisibilityDuringGame "NewEnumerator1"))

(event Custom|HideWhenReceivingFrom
  (params (Instigator :type Controller%20Object%20Reference))
  ;[0.4909 0 1 0.25] Gameplay Functions
  (CallFunction|SetForcedVisibilityDuringGame "NewEnumerator2"))

(event Custom|Restarted
  (params (Instigator :type Controller%20Object%20Reference))
  ;[0.4909 0 1 0.25] Gameplay Functions
  (switch Utilities|FlowControl|Switch|SwitchonE_TimerObjective_State (Variables|Transient|GetState)
    (:NewEnumerator1
      (CallFunction|SetState "NewEnumerator3" Instigator))))

(event OnAddedtoMinigame(FortMinigameLogic)
  (params (Minigame :type Fort%20Minigame%20Object%20Reference))
  (Networking|SwitchHasAuthority
    (:Authority
      ;[1 1 1] Player removed needed so we can clear reference if the referenced player leaves the game
      (Minigame|BindEventtoPlayerRemovedDelegate Minigame (EventDispatchers|CreateEvent :!event OnPlayerRemovedFromMinigame)))))

(event OnRemovedfromMinigame(FortMinigameLogic)
  (params (Minigame :type Fort%20Minigame%20Object%20Reference))
  (Networking|SwitchHasAuthority
    (:Authority
      (Minigame|UnbindEventfromPlayerRemovedDelegate Minigame (EventDispatchers|CreateEvent :!event OnPlayerRemovedFromMinigame)))))

(event MinigameEndedDelegate(FortMinigameLogic)
  (Networking|SwitchHasAuthority
    (:Authority
      ;[1 1 1] Clear Last Instigator
      (Variables|Transient|SetLastInteractingController 0))))

(event Custom|TriggerActivateVFX
  ;[1 0 0] Notifies
  (Variables|Colors|SetEmissiveActiveColor (Variables|Colors|GetTriggerActivateColorEmissive))
  (Effects|Components|ParticleSystem|SetVectorParameter (Variables|Default|GetP_TimerObjective_BeaconLight) "Color" (Math|Conversions|ToVector(LinearColor) (Variables|Colors|GetTriggerActivateColorParticles)))
  (CallFunction|PlayTriggerPulseVFXTimeline))

(event Custom|PlayTriggerPulseVFXTimeline
  ;[0 1 0] Timeline / Pulse Emissive & Play Particles
  (bind (_direction _lerp) (Timeline|Timeline_0 PlayFromStart)
    (:Update
      (Rendering|Material|SetCustomPrimitiveDataVector3 (Variables|Debug|GetStaticMeshComponent) (Variables|Material|GetEmissiveCustomPrimitiveIndex) (Math|Vector|Lerp(Vector) (Math|Conversions|ToVector(LinearColor) (Variables|Colors|GetEmissiveIdleColor)) (Math|Conversions|ToVector(LinearColor) (Variables|Colors|GetEmissiveActiveColor)) _lerp)))
    (:Spawn
      (Components|Activation|Activate (Variables|Default|GetP_TimerObjective_BeaconLight) true))))

(event Custom|TriggerDeactivateVFX
  ;[1 0 0] Notifies
  (Variables|Colors|SetEmissiveActiveColor (Variables|Colors|GetTriggerDeactiveColorEmissive))
  (Effects|Components|ParticleSystem|SetVectorParameter (Variables|Default|GetP_TimerObjective_BeaconLight) "Color" (Math|Conversions|ToVector(LinearColor) (Variables|Colors|GetTriggerDeactivateColorParticles)))
  (CallFunction|PlayTriggerPulseVFXTimeline))

(event Custom|TriggerPauseVFX
  ;[1 0 0] Notifies
  (Variables|Colors|SetEmissiveActiveColor (Variables|Colors|GetTriggerPauseColorEmissive))
  (Effects|Components|ParticleSystem|SetVectorParameter (Variables|Default|GetP_TimerObjective_BeaconLight) "Color" (Math|Conversions|ToVector(LinearColor) (Variables|Colors|GetTriggerPauseColorParticles)))
  (CallFunction|PlayTriggerPulseVFXTimeline))

(event Custom|TriggerCompletedVFX
  ;[1 0 0] Notifies
  (Variables|Colors|SetEmissiveActiveColor (Variables|Colors|GetTriggerCompletedColorEmissive))
  (Effects|Components|ParticleSystem|SetVectorParameter (Variables|Default|GetP_TimerObjective_BeaconLight) "Color" (Math|Conversions|ToVector(LinearColor) (Variables|Colors|GetTriggerCompletedColorParticles)))
  (CallFunction|PlayTriggerPulseVFXTimeline))

(event Custom|TriggerUrgentVFX
  ;[1 0 0] Notifies
  (Variables|Colors|SetEmissiveActiveColor (Variables|Colors|GetTriggerUrgentColorEmissive))
  (Effects|Components|ParticleSystem|SetVectorParameter (Variables|Default|GetP_TimerObjective_BeaconLight) "Color" (Math|Conversions|ToVector(LinearColor) (Variables|Colors|GetTriggerUrgentColorParticles)))
  (CallFunction|PlayTriggerPulseVFXTimeline))

(fn BlueprintCanInteract
  (params (InteractingPawn :type Fort%20Pawn%20Object%20Reference) (InteractionBeingAttempted :type EInteractionBeingAttempted%20Enum :default "FirstInteraction") (InteractionType :type TInteractionType%20Enum :default "IT_NoInteraction"))
  (bind _interactaction (Interaction|GetInteractionAction InteractingPawn))
  (switch Utilities|FlowControl|Switch|SwitchonE_TimerObjective_InteractAction _interactaction
    (:NewEnumerator6
      (return false))
    (:NewEnumerator5
      (return (Interaction|CanEditOptionsComponent (Variables|Default|GetToyOptionsComponent) InteractingPawn)))
    (:NewEnumerator0
      _)
    (:NewEnumerator1
      _)
    (:NewEnumerator2
      _)
    (:NewEnumerator3
      _)
    (:NewEnumerator4
      _))
  (return true))

(fn BlueprintGetInteractionString
  (params (InteractingPawn :type Fort%20Pawn%20Object%20Reference) (InteractionBeingAttempted :type EInteractionBeingAttempted%20Enum :default "FirstInteraction"))
  (bind (_designtext _activatetext _deactivatetext _state _timerhandle _activateinteracttime _deactivateinteracttime _time _activationscore _completedscore _deactivationscore _reactivatetext _pausetext _unpausetext) (Class|GetClassDefaults "/Game/Creative/Devices/TimerObjective/BP_TimerObjective.BP_TimerObjective_C"))
  (bind _interactaction (Interaction|GetInteractionAction InteractingPawn))
  (LocalVariables|SetOut (Utilities|Select :Index _interactaction :NewEnumerator5 (Variables|Text|GetDesignText) :NewEnumerator0 (Variables|Text|GetActivateText) :NewEnumerator1 (Variables|Text|GetDeactivateText) :NewEnumerator2 (Variables|Text|GetReactivateText) :NewEnumerator3 (Variables|Text|GetPauseText) :NewEnumerator4 (Variables|Text|GetUnpauseText)))
  (Utilities|FlowControl|Sequence
    (:then_0
      (if (Utilities|Text|TextIsEmpty (LocalVariables|GetOut))
        ;[1 1 1] Toy option defaults these variables to empty strings, which is the desired behaviour, so if its empty replace the interaction string with localizable text.
        (bind _returnvalue (|CallParentFunction (Variables|GetInteractingPawn) (Variables|GetInteractionBeingAttempted)))
        (LocalVariables|SetOut (Utilities|Select :Index _interactaction :NewEnumerator6 _returnvalue :NewEnumerator5 _designtext :NewEnumerator0 _activatetext :NewEnumerator1 _deactivatetext :NewEnumerator2 _reactivatetext :NewEnumerator3 _pausetext :NewEnumerator4 _unpausetext)) ; If text is empty, use the default values for the class
      ))
    (:then_1
      (return (LocalVariables|GetOut)))))

(fn BlueprintGetInteractionTime
  (params (InteractingPawn :type Fort%20Pawn%20Object%20Reference) (InteractionBeingAttempted :type EInteractionBeingAttempted%20Enum :default "FirstInteraction"))
  (bind _interactaction (Interaction|GetInteractionAction InteractingPawn))
  (return (Math|Conversions|ToFloat(Integer) (Utilities|Select :Index _interactaction :NewEnumerator6 0 :NewEnumerator5 0 :NewEnumerator0 (Variables|ToyProperties|GetActivateInteractTime) :NewEnumerator1 (Variables|ToyProperties|GetDeactivateInteractTime) :NewEnumerator2 (Variables|ToyProperties|GetReactivateInteractTime) :NewEnumerator3 (Variables|ToyProperties|GetPauseInteractTime) :NewEnumerator4 (Variables|ToyProperties|GetUnpauseInteractTime))) true))

(fn GetDisplayTextStyle
  (return (Utilities|String|StringToName (Variables|ToyProperties|GetTimerTextStyle))))

(fn GetLabelText
  (return (Variables|ToyProperties|GetTimerText)))

(fn IsHUDVisible
  (return (Math|Conversions|ToBoolean(Integer) (Variables|ToyProperties|GetIsVisibleOnHUD))))

(fn CanPause
  (return (not (Knobs|Team|IsTeamNone (Variables|ToyProperties|GetPauseTeamFilter)))))

(fn GetOverrideMaterialforAdditionalMesh
  (params (Mesh%20Cmp :type Static%20Mesh%20Component%20Object%20Reference) (Mat%20Idx :type Integer))
  (if (CallFunction|IsHologram)
    (return (Utilities|Select :Index (Variables|ToyProperties|GetMeshOptions) :NewEnumerator1 (Variables|Material|AdditionalMesh|GetExplosiveDevice_HologramMaterial)))
    (else
      (return (Utilities|Select :Index (Variables|ToyProperties|GetMeshOptions) :NewEnumerator1 (Variables|Material|AdditionalMesh|GetExplosiveDevice_NormalMaterial))))))

(fn GetOverrideMaterialforBaseMesh
  (params (Mesh%20Cmp :type Static%20Mesh%20Component%20Object%20Reference) (Mat%20Idx :type Integer))
  (if (CallFunction|IsHologram)
    (return (Variables|Material|GetHologramMaterial))
    (elif (== Mat%20Idx 0)
      (return (Variables|Material|GetNormalMaterial))
      (elif (== Mat%20Idx 1)
        (return (Variables|Material|GetLCDMaterial))
        (else
          (return 0))))))

(fn GetInteractActionFromStates
  (if (and (Utilities|Enum|Equal(Enum) (Variables|Transient|GetLastState) "NewEnumerator7") (Utilities|Enum|Equal(Enum) (Variables|Transient|GetState) "NewEnumerator3"))
    (return "NewEnumerator0")
    (elif (and (Utilities|Enum|Equal(Enum) (Variables|Transient|GetLastState) "NewEnumerator1") (Utilities|Enum|Equal(Enum) (Variables|Transient|GetState) "NewEnumerator3"))
      (return "NewEnumerator2")
      (elif (and (Utilities|Enum|Equal(Enum) (Variables|Transient|GetLastState) "NewEnumerator3") (Utilities|Enum|Equal(Enum) (Variables|Transient|GetState) "NewEnumerator4"))
        (return "NewEnumerator3")
        (elif (and (Utilities|Enum|Equal(Enum) (Variables|Transient|GetLastState) "NewEnumerator3") (Utilities|Enum|Equal(Enum) (Variables|Transient|GetState) "NewEnumerator1"))
          (return "NewEnumerator1")
          (elif (and (Utilities|Enum|Equal(Enum) (Variables|Transient|GetLastState) "NewEnumerator4") (Utilities|Enum|Equal(Enum) (Variables|Transient|GetState) "NewEnumerator3"))
            (return "NewEnumerator4")
            (elif (and (Utilities|Enum|Equal(Enum) (Variables|Transient|GetLastState) "NewEnumerator4") (Utilities|Enum|Equal(Enum) (Variables|Transient|GetState) "NewEnumerator1"))
              (return "NewEnumerator1")
              (else
                (return "NewEnumerator6")))))))))

(fn GetInteractionAction
  (params (InteractingPawn :type Pawn%20Object%20Reference))
  (if (Utilities|Time|IsTimerActivebyHandle (Variables|Transient|GetInteractionLockTimerHandle))
    (return "NewEnumerator6")
    (else
      (Utilities|FlowControl|Sequence
        (:then_0
          ;[1 1 1] Set Interacting Team if possible, otherwise leave as INDEX_NONE
          (bind _interactingpawn (Variables|GetInteractingPawn)
            (:then
              (if (.ReturnValue (Team|GetActorPvPTeamIndex _interactingpawn))
                (LocalVariables|SetInteractingTeam (Math|Conversions|ToInteger(Byte) (Team|GetActorPvPTeamIndex _interactingpawn)))
                (else
                  (LocalVariables|SetInteractingTeam (Utilities|INDEXNONE)))))
            (:else)))
        (:then_1
          (switch Utilities|FlowControl|Switch|SwitchonE_TimerObjective_State (Variables|Transient|GetState)
            (:NewEnumerator0
              (LocalVariables|SetOut "NewEnumerator5"))
            (:NewEnumerator7
              ;[0.08939 1 0.1757] Is the player allowed to ACTIVATE this timer objective?
              (if (CallFunction|PassesTeamFilter (Variables|ToyProperties|GetActivateTeamFilter) (LocalVariables|GetInteractingTeam))
                (LocalVariables|SetOut "NewEnumerator0")))
            (:NewEnumerator1
              ;[1 0.1042 0.1042] Is the player allowed to REACTIVATE this timer objective?
              (if (CallFunction|PassesTeamFilter (Variables|ToyProperties|GetReactivateTeamFilter) (LocalVariables|GetInteractingTeam))
                (LocalVariables|SetOut "NewEnumerator2")))
            (:NewEnumerator3
              (if (CallFunction|CanPause)
                ;[1 0.9913 0.1407] Is the player allowed to PAUSE this timer objective?
                (if (CallFunction|PassesTeamFilter (Variables|ToyProperties|GetPauseTeamFilter) (LocalVariables|GetInteractingTeam))
                  (LocalVariables|SetOut "NewEnumerator3")
                  (elif (CallFunction|PassesTeamFilter (Variables|ToyProperties|GetDeactivateTeamFilter) (LocalVariables|GetInteractingTeam))
                    ;[0.09063 0.04957 1] Is the player allowed to DEACTIVATE this timer objective?
                    (LocalVariables|SetOut "NewEnumerator1")))
                (else
                  _)))
            (:NewEnumerator4
              (if (CallFunction|CanPause)
                ;[1 0.3354 0.03153] Is the player allowed to UNPAUSE this timer objective?
                (if (CallFunction|PassesTeamFilter (Variables|ToyProperties|GetUnpauseTeamFilter) (LocalVariables|GetInteractingTeam))
                  (LocalVariables|SetOut "NewEnumerator4")
                  (else
                    _))
                (else
                  _)))))
        (:then_2
          (return (LocalVariables|GetOut)))))))

(fn GetOverrideMeshMaterial
  (params (MeshComp :type Static%20Mesh%20Component%20Object%20Reference) (MatIdx :type Integer))
  (bind _matidx (Variables|GetMatIdx))
  (bind _meshcomp (Variables|GetMeshComp))
  (bind _material (CallFunction|GetOverrideMaterialForBaseMesh _meshcomp _matidx))
  (bind _matidx_2 (Variables|GetMatIdx))
  (bind _meshcomp_2 (Variables|GetMeshComp))
  (bind _material_2 (CallFunction|GetOverrideMaterialForAdditionalMesh _meshcomp_2 _matidx_2))
  ;[1 1 1] Base Mesh
  (if (Utilities|Equal(Object) MeshComp (Variables|Debug|GetStaticMeshComponent))
    (Utilities|IsValid _material
      (:Is%20Valid
        (return _material))
      (:Is%20Not%20Valid
        (return (|CallParentFunction (Variables|GetMeshComp) (Variables|GetMatIdx)))))
    (elif (Utilities|Equal(Object) MeshComp (Variables|Default|GetAdditional%20Mesh))
      ;[1 1 1] Additional Mesh
      (Utilities|IsValid _material_2
        (:Is%20Valid
          (return _material_2))
        (:Is%20Not%20Valid
          (return (|CallParentFunction (Variables|GetMeshComp) (Variables|GetMatIdx)))))
      (else
        _))))

(fn HandleOnInteract
  (params (InteractingPawn :type Pawn%20Object%20Reference))
  (bind _interactaction (Interaction|GetInteractionAction InteractingPawn))
  (bind _output_get (LocalVariables|SetInteractionAction _interactaction))
  (if (Utilities|Enum|NotEqual(Enum) _output_get "NewEnumerator6")
    (CallFunction|SetState (Utilities|Select :Index (LocalVariables|GetInteractionAction) :NewEnumerator6 "NewEnumerator0" :NewEnumerator5 "NewEnumerator0" :NewEnumerator0 "NewEnumerator3" :NewEnumerator1 "NewEnumerator1" :NewEnumerator2 "NewEnumerator3" :NewEnumerator3 "NewEnumerator4" :NewEnumerator4 "NewEnumerator3") (Pawn|GetController (Variables|GetInteractingPawn)))))

(fn IsHologram
  (LocalVariables|SetbIsHologram false)
  (Utilities|FlowControl|Sequence
    (:then_0
      (if (Math|Conversions|ToBoolean(Integer) (Variables|ToyProperties|GetHologramUntilActivated))
        (if (Utilities|Enum|Equal(Enum) (Variables|Transient|GetState) "NewEnumerator7")
          (LocalVariables|SetbIsHologram true)
          (elif (Utilities|Enum|Equal(Enum) (Variables|Transient|GetState) "NewEnumerator0")
            (if (not (Math|Conversions|ToBoolean(Integer) (Variables|ToyProperties|GetActivatedOnGameStart)))
              (bind _fortminigamelogic (Variables|Default|GetFortMinigameLogic)
                (:then
                  (bind _currentminigame (Variables|/Script/FortniteGame.FortMinigameLogicComponent|GetCurrentMinigame _fortminigamelogic)
                    (:then
                      (if (Minigame|IsinPlayMode _currentminigame :!class /Script/FortniteGame.FortMinigame)
                        (LocalVariables|SetbIsHologram true)
                        (elif (Utilities|Enum|Equal(Enum) (Variables|/Script/FortniteGame.FortMinigame|GetCurrentState _currentminigame) "Transitioning")
                          (LocalVariables|SetbIsHologram true))))
                    (:else)))
                (:else)))))))
    (:then_1
      (return (LocalVariables|GetbIsHologram)))))

(fn MessageScore
  (return (Variables|UserOptions|Advanced|GetHUD%20Message) (Variables|UserOptions|Advanced|GetHUD%20MessageScoreColor) (Variables|UserOptions|Advanced|GetHUD%20MessageColor) (Variables|UserOptions|Advanced|GetResetHUD%20MessageScore)))

(fn OnClockTimerExpired
  (Utilities|FlowControl|Sequence
    (:then_0
      (Networking|CanExecuteCosmeticEvents
        (:True
          (CallFunction|UpdateVisualDisplays))
        (:False
          (CallFunction|UpdateMinimapTime))))
    (:then_1
      (Networking|CanExecuteCosmeticEvents
        (:True
          _)
        (:False
          (Utilities|ShouldUpdateClockonServer
            (:Update
              _))))
      (if (Utilities|Enum|Equal(Enum) (Variables|Transient|GetState) "NewEnumerator3")
        (bind _returnvalue (Utilities|Time|SetTimerbyEvent (EventDispatchers|CreateEvent :!event OnClockTimerExpired) (Variables|Material|GetClockTimerUpdateRate)))
        (Variables|Transient|SetClockTimerHandle _returnvalue)))))

(fn OnInteractionLockTimerExpired)

(fn OnMinigameStateChanged
  (params (Minigame :type Fort%20Minigame%20Object%20Reference) (NewMinigameState :type EFortMinigameState%20Enum :default "PreGame"))
  (Networking|SwitchHasAuthority
    (:Authority
      ;[0.5 0.25 0.5] Set Timer Objective State
      (if (Utilities|Enum|Equal(Enum) (Variables|GetNewMinigameState) "InProgress")
        (if (Math|Conversions|ToBoolean(Integer) (Variables|ToyProperties|GetActivatedOnGameStart))
          (CallFunction|SetState "NewEnumerator3")
          (else
            (CallFunction|SetState "NewEnumerator7")))
        (else
          (CallFunction|SetState "NewEnumerator0"))))))

(fn OnPlayerRemovedFromMinigame
  (params (Unique%20Net%20ID :type Unique%20Net%20Id%20Structure) (bIsLocalPlayer :type Boolean))
  (bind _lastinteractingcontroller (Variables|Transient|GetLastInteractingController))
  (Networking|SwitchHasAuthority
    (:Authority
      (Utilities|IsValid _lastinteractingcontroller
        (:Is%20Valid
          (bind _asfort_player_state (Utilities|Casting|CastToFortPlayerState (Variables|/Script/Engine.Controller|GetPlayerState _lastinteractingcontroller)))
          (if (== (PlayerState|GetPlayerUniqueNetID _asfort_player_state) Unique%20Net%20ID)
            (if (Utilities|Enum|Equal(Enum) (Variables|Default|GetInstigatingPlayerOption) "NewEnumerator0")
              ;[1 1 1] Clear Last Instigator
              (Variables|Transient|SetLastInteractingController 0)
              (else
                (bind _playercontroller (Utilities|GetRandomPlayer)
                  (:HasValidMember
                    (Variables|Transient|SetLastInteractingController _playercontroller))
                  (:NoValidMember
                    (Variables|Transient|SetLastInteractingController 0)))))))))))

(fn OnRep_MinimapTime
  (MiniMapData|SetIndicatorText (Variables|Default|GetCreativeMapMarkerComponent) (Variables|Transient|GetMinimapTime)))

(fn OnRep_State
  (bind _hologram_audio_loop (Variables|Default|GetHologram%20Audio%20Loop))
  (Utilities|FlowControl|Sequence
    (:then_0
      (switch Utilities|FlowControl|Switch|SwitchonE_TimerObjective_State (Variables|Transient|GetState)
        (:NewEnumerator0
          ;[1 0 0] Entering Design
          (Utilities|Time|ClearandInvalidateTimerbyHandle (Variables|Transient|GetTimerHandle))
          (Utilities|Time|ClearandInvalidateTimerbyHandle (Variables|Transient|GetClockTimerHandle))
          (Audio|Components|Audio|Stop (Variables|Default|GetHologram%20Audio%20Loop)))
        (:NewEnumerator7
          (Networking|CanExecuteCosmeticEvents
            (:True
              (CallFunction|TriggerDeactivateVFX))))
        (:NewEnumerator1
          ;[0 1 0] Entering Deactivated
          (Utilities|Time|ClearandInvalidateTimerbyHandle (Variables|Transient|GetTimerHandle))
          (Utilities|Time|ClearandInvalidateTimerbyHandle (Variables|Transient|GetClockTimerHandle))
          (Utilities|FlowControl|Sequence
            (:then_0
              (Networking|SwitchHasAuthority
                (:Authority
                  (GameplayEvent|BroadcastEvent (Variables|UserOptions|Events|GetOn%20Stopped) 0 (Utilities|Struct|MakeGameplayEventDeactivated (Variables|Transient|GetLastInteractingController))))))
            (:then_1
              (if (Math|Conversions|ToBoolean(Integer) (Variables|ToyProperties|GetPlayDeactivationSound))
                (CallFunction|PlaySound (Variables|Audio|GetDeactivationSound) (Math|Conversions|ToBoolean(Integer) (Variables|ToyProperties|GetDeactivationSoundAttenuation)))))
            (:then_2
              (CallFunction|PlayMessage (Variables|ToyProperties|GetDeactivationText))
              (Networking|CanExecuteCosmeticEvents
                (:True
                  (CallFunction|PlayParticle (Variables|VFX|GetDeactivatedParticleTemplate))
                  (CallFunction|TriggerDeactivateVFX)
                  (Audio|Components|Audio|Stop (Variables|Default|GetPlant%20Power%20Up%20Audio%20Loop)))))))
        (:NewEnumerator3
          (if (and (Utilities|Time|DoesTimerExistbyHandle (Variables|Transient|GetTimerHandle)) (> (Utilities|Time|GetTimerRemainingTimebyHandle (Variables|Transient|GetTimerHandle)) 0.5))
            ;[0 0 1] Entering Activated (From Paused)
            (Utilities|Time|UnpauseTimerbyHandle (Variables|Transient|GetTimerHandle))
            (Utilities|FlowControl|Sequence
              (:then_0
                (Networking|SwitchHasAuthority
                  (:Authority
                    (GameplayEvent|BroadcastEvent (Variables|UserOptions|Events|GetOn%20Resumed) 0 (Utilities|Struct|MakeGameplayEventUnpaused (Variables|Transient|GetLastInteractingController)))
                    (Variables|Transient|SetTimerFinishedServerTimeSeconds (+ (Variables|Transient|GetTimerFinishedServerTimeSeconds) (- (Utilities|Time|GetGameTimeinSeconds) (Variables|Transient|GetPausedTimeSeconds)))))))
              (:then_1
                (CallFunction|PlayMessage (Variables|ToyProperties|GetUnpausedText))))
            (else
              ;[1 1 0] Entering Activated
              (bind _returnvalue (Utilities|Time|SetTimerbyEvent (EventDispatchers|CreateEvent :!event OnTimerExpired) (Math|Conversions|ToFloat(Integer) (Variables|ToyProperties|GetTime))))
              (Variables|Transient|SetTimerHandle _returnvalue)
              (Utilities|FlowControl|Sequence
                (:then_0
                  (Networking|SwitchHasAuthority
                    (:Authority
                      (GameplayEvent|BroadcastEvent (Variables|UserOptions|Events|GetOn%20Started) 0 (Utilities|Struct|MakeGameplayEventActivated (Variables|Transient|GetLastInteractingController)))
                      (if (Utilities|Enum|Equal(Enum) (Variables|Transient|GetLastState) "NewEnumerator1")
                        (GameplayEvent|BroadcastEvent (Variables|UserOptions|Events|GetOn%20Restarted) 0 (Utilities|Struct|MakeGameplayEventRestarted (Variables|Transient|GetLastInteractingController)))
                        (else
                          _))
                      (Variables|Transient|SetTimerFinishedServerTimeSeconds (+ (Utilities|Time|GetTimeSeconds) (Math|Conversions|ToFloat(Integer) (Variables|ToyProperties|GetTime))))
                      (Variables|Transient|SetArmedFinishedServerTimeSeconds (+ (Utilities|Time|GetTimeSeconds) (Variables|Transient|GetArmedTimeSeconds))))))
                (:then_1
                  (if (Math|Conversions|ToBoolean(Integer) (Variables|ToyProperties|GetPlayActivationSound))
                    (CallFunction|PlaySound (Variables|Audio|GetActivationSound) (Math|Conversions|ToBoolean(Integer) (Variables|ToyProperties|GetActivationSoundAttenuation)))))
                (:then_2
                  ;[0 1 1] Entering Paused
                  (CallFunction|PlayMessage (Variables|ToyProperties|GetActivationText))))))
          (bind _returnvalue_2 (Utilities|Time|SetTimerbyEvent (EventDispatchers|CreateEvent :!event OnClockTimerExpired) (Variables|Material|GetClockTimerUpdateRate)))
          (Variables|Transient|SetClockTimerHandle _returnvalue_2)
          (Networking|CanExecuteCosmeticEvents
            (:True
              (CallFunction|PlayParticle (Variables|VFX|GetActivatedParticleTemplate))
              (CallFunction|TriggerActivateVFX)
              (Audio|Components|Audio|Stop (Variables|Default|GetPlant%20Power%20Up%20Audio%20Loop)))))
        (:NewEnumerator4
          (if (Utilities|Time|DoesTimerExistbyHandle (Variables|Transient|GetTimerHandle))
            (Utilities|Time|PauseTimerbyHandle (Variables|Transient|GetTimerHandle))
            (Utilities|FlowControl|Sequence
              (:then_0
                (Networking|SwitchHasAuthority
                  (:Authority
                    (GameplayEvent|BroadcastEvent (Variables|UserOptions|Events|GetOn%20Paused) 0 (Utilities|Struct|MakeGameplayEventPaused (Variables|Transient|GetLastInteractingController)))
                    (Variables|Transient|SetPausedTimeSeconds (Utilities|Time|GetGameTimeinSeconds)))))
              (:then_1
                (CallFunction|PlayMessage (Variables|ToyProperties|GetPausedText))
                (Utilities|Time|ClearandInvalidateTimerbyHandle (Variables|Transient|GetClockTimerHandle))
                (Networking|CanExecuteCosmeticEvents
                  (:True
                    (CallFunction|TriggerPauseVFX)
                    (Audio|Components|Audio|Stop (Variables|Default|GetPlant%20Power%20Up%20Audio%20Loop))))))))
        (:NewEnumerator5
          ;[1 0 1] Entering Completed
          (Utilities|FlowControl|Sequence
            (:then_0
              (Networking|SwitchHasAuthority
                (:Authority
                  (GameplayEvent|BroadcastEvent (Variables|UserOptions|Events|GetOn%20Completed) 0 (Utilities|Struct|MakeGameplayEventCompleted (Variables|Transient|GetLastInteractingController))))))
            (:then_1
              (if (Math|Conversions|ToBoolean(Integer) (Variables|ToyProperties|GetPlayCompletedSound))
                (CallFunction|PlaySound (Variables|Audio|GetCompletionSound) (Math|Conversions|ToBoolean(Integer) (Variables|ToyProperties|GetCompletedSoundAttenuation)))))
            (:then_2
              (CallFunction|PlayMessage (Variables|ToyProperties|GetCompletedText))
              (Networking|CanExecuteCosmeticEvents
                (:True
                  (CallFunction|PlayParticle (Variables|VFX|GetCompletedParticleTemplate))
                  (CallFunction|TriggerCompletedVFX)
                  (Audio|Components|Audio|Stop (Variables|Default|GetPlant%20Power%20Up%20Audio%20Loop)))))))
        (:NewEnumerator10
          ;[0.5 0.5 1] Entering Disabled
          (Utilities|Time|ClearandInvalidateTimerbyHandle (Variables|Transient|GetTimerHandle))
          (Utilities|Time|ClearandInvalidateTimerbyHandle (Variables|Transient|GetClockTimerHandle))
          (Networking|CanExecuteCosmeticEvents
            (:True
              (CallFunction|TriggerCompletedVFX)
              (Audio|Components|Audio|Stop (Variables|Default|GetPlant%20Power%20Up%20Audio%20Loop)))))))
    (:then_1
      (CallFunction|UpdateMeshesCollisionProfiles)
      (if (Math|Conversions|ToBoolean(Integer) (Variables|ToyProperties|GetSounds))
        (if (CallFunction|IsHologram)
          (Audio|Components|Audio|Play _hologram_audio_loop 0.0)
          (else
            (Audio|Components|Audio|Stop _hologram_audio_loop)))))
    (:then_2
      (Networking|CanExecuteCosmeticEvents
        (:True
          (CallFunction|UpdateVisualDisplays)
          (CallFunction|UpdateAdditionalMeshes))
        (:False
          (CallFunction|UpdateMinimapTime))))
    (:then_3
      (Networking|SwitchHasAuthority
        (:Remote
          (bind _returnvalue_3 (Utilities|Time|SetTimerbyEvent (EventDispatchers|CreateEvent :!event OnInteractionLockTimerExpired) (Variables|Default|GetInteractionLockTime)))
          (Variables|Transient|SetInteractionLockTimerHandle _returnvalue_3))))))

(fn OnRep_Time
  (Networking|CanExecuteCosmeticEvents
    (:True
      (CallFunction|UpdateVisualDisplays))
    (:False
      (CallFunction|UpdateMinimapTime))))

(fn OnTimerExpired
  (Networking|SwitchHasAuthority
    (:Authority
      (CallFunction|SetState "NewEnumerator5" (Variables|Transient|GetLastInteractingController)))))

(fn PassesTeamFilter
  (params (Team%20Filter :type Integer) (Interacting%20Team :type Integer))
  (return (Math|Boolean|ORBoolean (== (Variables|GetInteracting%20Team) (Utilities|INDEXNONE)) (Knobs|Team|IsTeamAll (Variables|GetTeam%20Filter)) (== (Variables|GetInteracting%20Team) (Knobs|Team|GetTeamValueasIndex (Variables|GetTeam%20Filter))))))

(fn PlayMessage
  (params (Text :type Text))
  (bind _returnvalue (Utilities|Enum|LiteralenumEAthenaGameMsgType "DefaultIntro"))
  (Networking|CanExecuteCosmeticEvents
    (:True
      (if (not (Utilities|Text|TextIsEmpty Text))
        (bind _asfort_game_state_athena (Utilities|Casting|CastToFortGameStateAthena (Game|GetGameState)))
        (UI|Messaging|HandleHideGameModeMessageRequest _asfort_game_state_athena _returnvalue false false -1)
        (UI|Messaging|HandleGameModeMessageRequest _asfort_game_state_athena (Utilities|Struct|MakeAthenaGameMessageData _returnvalue Text 0 0 0.0 false 0 3.0))))))

(fn PlayParticle
  (params (ParticleTemplate :type Particle%20System%20Object%20Reference))
  (Networking|CanExecuteCosmeticEvents
    (:True
      (Utilities|IsValid ParticleTemplate
        (:Is%20Valid
          (Effects|SpawnEmitteratLocation ParticleTemplate (Transformation|GetActorLocation) (Transformation|GetActorRotation) "1.000000,1.000000,1.000000" true "AutoRelease"))))))

(fn PlaySound
  (params (Sound :type Sound%20Base%20Object%20Reference) (bAttenutation :type Boolean :default true))
  (Networking|CanExecuteCosmeticEvents
    (:True
      (if (Math|Conversions|ToBoolean(Integer) (Variables|ToyProperties|GetSounds))
        (Utilities|IsValid (Variables|GetSound)
          (:Is%20Valid
            (Audio|PlaySoundatLocation (Variables|GetSound) (Transformation|GetActorLocation) (Transformation|GetActorRotation) 1.0 1.0 0.0 (select (Variables|GetbAttenutation) 0 "/Game/Sounds/Attenuation/Fort_Creative_NoAttenuation.Fort_Creative_NoAttenuation") 0 self)))))))

(fn RefreshFoliage
  (params (bShowInGame :type Boolean))
  (Networking|CanExecuteCosmeticEvents
    (:True
      (if bShowInGame
        (Utilities|Array|Clear (Variables|BuildingActor|GetOverridePrimitivesToExcludeFoliage))
        (else
          (Variables|BuildingActor|SetOverridePrimitivesToExcludeFoliage (Utilities|Array|MakeArray (Variables|Default|GetNonCollidingSphere)))))
      (Building|RefreshRemovedFoliageAroundBounds false))))

(fn RefreshMeshVisibilities
  (LocalVariables|SetDesiredVisibility (select (Math|Conversions|ToBoolean(Integer) (Variables|ToyProperties|GetIsVisibleDuringGame)) "NewEnumerator3" "NewEnumerator1"))
  (Utilities|FlowControl|Sequence
    (:then_0
      ;[0.5 0.25 0.5] Main Mesh
      (Visibility|SetVisiblityDuringPhase (Variables|Default|GetStaticMeshVisibleInGameComponent) (LocalVariables|GetDesiredVisibility)))
    (:then_1
      ;[0.25 0.5 0.5] Additional Mesh
      (Visibility|SetVisiblityDuringPhase (Variables|Default|GetAdditionalMeshesVisibleInGameComponent) (select (Utilities|Enum|Equal(Enum) (Variables|ToyProperties|GetMeshOptions) "NewEnumerator1") "NewEnumerator3" "NewEnumerator0")))))

(fn SetForcedVisibilityDuringGame
  (params (NewForcedVisibility :type Enum_ForcedValueDuringGame%20Enum :default "NewEnumerator0"))
  (Utilities|FlowControl|Sequence
    (:then_0
      ;[0.5 0.25 0.5] Main Mesh
      (Visibility|SetForcedVisibilityDuringGame (Variables|Default|GetStaticMeshVisibleInGameComponent) (Variables|GetNewForcedVisibility)))
    (:then_1
      ;[0.25 0.5 0.5] Additional Mesh
      (Visibility|SetForcedVisibilityDuringGame (Variables|Default|GetAdditionalMeshesVisibleInGameComponent) (select (Utilities|Enum|Equal(Enum) (Variables|ToyProperties|GetMeshOptions) "NewEnumerator0") "NewEnumerator2" (Variables|GetNewForcedVisibility))))))

(fn SetState
  (params (NewState :type E_TimerObjective_State%20Enum :default "NewEnumerator0") (New%20Instigator%20Controller :type Controller%20Object%20Reference))
  (bind (_outsourcetags _outcontexttags) (Class|FortMinigameStatsComponent|GetSourceandContextTags (LocalVariables|Getlocal%20MinigameStatsComp) (LocalVariables|Getlocal%20FortPlayerController)))
  (Networking|SwitchHasAuthority
    (:Authority
      (if (Utilities|Enum|NotEqual(Enum) NewState (Variables|Transient|GetState))
        (Variables|Transient|SetLastState (Variables|Transient|GetState))
        ;[1 1 1] If going to Design state, ensure the mini game is not in progress
        (if (Utilities|Enum|Equal(Enum) NewState "NewEnumerator0")
          (bind _currentminigame (Variables|/Script/FortniteGame.FortMinigameLogicComponent|GetCurrentMinigame (Variables|Default|GetFortMinigameLogic))
            (:then
              (Minigame|IsInProgress _currentminigame
                (:No
                  _)))
            (:else))
          (else
            _))
        ;[1 1 1] If pausing ensure this is pausable
        (if (Utilities|Enum|Equal(Enum) NewState "NewEnumerator4")
          (if (CallFunction|CanPause)
            _)
          (else
            _))
        (Variables|Transient|SetLastInteractingController (Variables|GetNew%20Instigator%20Controller))
        (Variables|Transient|SetState NewState)
        (Utilities|FlowControl|Sequence
          (:then_0
            ;[1 0 0] Give the player controller score
            (bind _new_instigator_controller (Variables|GetNew%20Instigator%20Controller)
              (:then
                (bind _asfort_player_controller (Utilities|Casting|CastToFortPlayerController _new_instigator_controller))
                (LocalVariables|Setlocal%20FortPlayerController _asfort_player_controller)
                (bind _local_fortplayercontroller (LocalVariables|Getlocal%20FortPlayerController)
                  (:then
                    (bind _returnvalue (CreativeDevices|GetMinigameStatsComp _local_fortplayercontroller "Athena"))
                    (LocalVariables|Setlocal%20MinigameStatsComp _returnvalue)
                    (LocalVariables|Getlocal%20MinigameStatsComp
                      (:then
                        (bind _output_get (LocalVariables|SetScore (Utilities|Select :Index (Variables|Transient|GetState) :NewEnumerator0 0 :NewEnumerator7 (Utilities|Select :Index (CallFunction|GetInteractActionfromStates) :NewEnumerator6 0 :NewEnumerator5 0 :NewEnumerator0 (Variables|ToyProperties|GetActivationScore) :NewEnumerator1 (Variables|ToyProperties|GetDeactivationScore) :NewEnumerator2 (Variables|ToyProperties|GetReactivationScore) :NewEnumerator3 (Variables|ToyProperties|GetPauseScore) :NewEnumerator4 (Variables|ToyProperties|GetUnpauseScore)) :NewEnumerator1 (Utilities|Select :Index (CallFunction|GetInteractActionfromStates) :NewEnumerator6 0 :NewEnumerator5 0 :NewEnumerator0 (Variables|ToyProperties|GetActivationScore) :NewEnumerator1 (Variables|ToyProperties|GetDeactivationScore) :NewEnumerator2 (Variables|ToyProperties|GetReactivationScore) :NewEnumerator3 (Variables|ToyProperties|GetPauseScore) :NewEnumerator4 (Variables|ToyProperties|GetUnpauseScore)) :NewEnumerator3 (Utilities|Select :Index (CallFunction|GetInteractActionfromStates) :NewEnumerator6 0 :NewEnumerator5 0 :NewEnumerator0 (Variables|ToyProperties|GetActivationScore) :NewEnumerator1 (Variables|ToyProperties|GetDeactivationScore) :NewEnumerator2 (Variables|ToyProperties|GetReactivationScore) :NewEnumerator3 (Variables|ToyProperties|GetPauseScore) :NewEnumerator4 (Variables|ToyProperties|GetUnpauseScore)) :NewEnumerator4 (Utilities|Select :Index (CallFunction|GetInteractActionfromStates) :NewEnumerator6 0 :NewEnumerator5 0 :NewEnumerator0 (Variables|ToyProperties|GetActivationScore) :NewEnumerator1 (Variables|ToyProperties|GetDeactivationScore) :NewEnumerator2 (Variables|ToyProperties|GetReactivationScore) :NewEnumerator3 (Variables|ToyProperties|GetPauseScore) :NewEnumerator4 (Variables|ToyProperties|GetUnpauseScore)) :NewEnumerator5 (Variables|ToyProperties|GetCompletedScore) :NewEnumerator10 0)))
                        (if (> _output_get 0)
                          (bind _output_get_2 (LocalVariables|SetTargetTags (GameplayTags|GetOwnedGameplayTags self :!class /Script/GameplayTags.GameplayTagAssetInterface)))
                          (GameplayTags|AddGameplayTag _output_get_2 "(TagName=\"Minigame.Stat.Score.Goal\")" :!class /Script/GameplayTags.BlueprintGameplayTagLibrary)
                          (Class|FortMinigameStatsComponent|SendStatEventwithTags (LocalVariables|Getlocal%20MinigameStatsComp) (LocalVariables|Getlocal%20FortPlayerController) "MinigameDynamicEvent" self (LocalVariables|GetTargetTags) _outsourcetags _outcontexttags (LocalVariables|GetScore))
                          (Class|CreativeHUDBroadcastComponent|DoScoreEvent (Variables|Default|GetCreative_HUDBroadcast_Component) (LocalVariables|Getlocal%20FortPlayerController) (LocalVariables|GetScore) (Variables|UserOptions|Advanced|GetDisplayScoreUpdateOnHUD))))
                      (:else)))
                  (:else)))
              (:else)))
          (:then_1
            ;[0 1 0] Handle completion behavior
            (if (Utilities|Enum|Equal(Enum) (Variables|Transient|GetState) "NewEnumerator5")
              (switch Utilities|FlowControl|Switch|SwitchonE_TimerObjective_OnCompleteBehavior (Variables|ToyProperties|GetOnCompletionBehavior)
                (:NewEnumerator0)
                (:NewEnumerator1
                  (CallFunction|SetState "NewEnumerator7" (Variables|Transient|GetLastInteractingController)))
                (:NewEnumerator2
                  (CallFunction|SetState "NewEnumerator3" (Variables|Transient|GetLastInteractingController))
                  (GameplayEvent|BroadcastEvent (Variables|UserOptions|Events|GetOn%20Restarted) 0 (Utilities|Struct|MakeGameplayEventRestarted (Variables|Transient|GetLastInteractingController))))))))))))

(fn UpdateAdditionalMeshes
  (if (Utilities|Enum|NotEqual(Enum) (Variables|ToyProperties|GetMeshOptions) "NewEnumerator0")
    (for _index (range (Rendering|Material|GetNumMaterials (Variables|Default|GetAdditional%20Mesh)))
      (Rendering|Material|SetMaterial (Variables|Default|GetAdditional%20Mesh) _index (CallFunction|GetOverrideMeshMaterial (Variables|Default|GetAdditional%20Mesh) _index)))))

(fn UpdateVisibilityComponents
  (Class|CreativeVisibleinGameComponent|InitializeComponent (Variables|Default|GetStaticMeshVisibleInGameComponent))
  (Visibility|SetVisibilityComponents (Variables|Default|GetAdditionalMeshesVisibleInGameComponent) (Utilities|Array|MakeArray (Variables|Default|GetAdditional%20Mesh)) false)
  (CallFunction|RefreshMeshVisibilities))

(fn UpdateHUDData
  (CallFunction|SetHUDData 0 (Utilities|Struct|MakeTimerObjectiveHUDData (select (or (Utilities|Enum|Equal(Enum) (Variables|Transient|GetState) "NewEnumerator3") (Utilities|Enum|Equal(Enum) (Variables|Transient|GetState) "NewEnumerator4")) (- (Variables|Transient|GetTimerFinishedServerTimeSeconds) (Fortnite|Sync|Time|GetServerWorldTimeSeconds)) 0.0) (Utilities|Enum|Equal(Enum) (Variables|Transient|GetState) "NewEnumerator4")))
  ;[0.25 0.5 0.25] To keep the LCD and the HUD in sync, the Timer Objective will control the signal when to update
  (CallFunction|BroadcastUpdatetoHUD))

(fn UpdateLCD
  (bind _timersoundattenuation (Variables|ToyProperties|GetTimerSoundAttenuation))
  (bind _returnvalue (Math|Conversions|ToBoolean(Integer) _timersoundattenuation))
  (bind (_minutes _seconds) (Utilities|GetTimeasMinutesSeconds))
  (bind (_returnvalue_r _returnvalue_g _returnvalue_b _returnvalue_a) (Math|Color|BreakColor (Utilities|Select :Index (Variables|Transient|GetState) :NewEnumerator0 (Variables|Colors|GetDeactivatedColor) :NewEnumerator7 (Variables|Colors|GetDeactivatedColor) :NewEnumerator1 (Variables|Colors|GetDeactivatedColor) :NewEnumerator3 (select (Math|Boolean|ANDBoolean (Math|Conversions|ToBoolean(Integer) (Variables|ToyProperties|GetAllowUrgencyMode)) (== (LocalVariables|GetLCDMinutes) 0) (<= (LocalVariables|GetLCDSeconds) (Variables|ToyProperties|GetUrgentModeStartTime))) (Variables|Colors|GetUrgentColor) (Variables|Colors|GetEnabledColor)) :NewEnumerator4 (Variables|Colors|GetPausedColor) :NewEnumerator5 (Variables|Colors|GetDisabledColor) :NewEnumerator10 (Variables|Colors|GetDisabledColor))))
  (Networking|CanExecuteCosmeticEvents
    (:True
      (Utilities|FlowControl|Sequence
        (:then_0
          ;[1 0 0] Calculate the time remaining in minutes and seconds
          (if (Utilities|Enum|Equal(Enum) (Variables|Transient|GetState) "NewEnumerator5")
            (LocalVariables|SetLCDMinutes 88)
            (LocalVariables|SetLCDSeconds 88)
            (else
              (LocalVariables|SetLCDMinutes _minutes)
              (LocalVariables|SetLCDSeconds _seconds))))
        (:then_1
          (if (or (!= (Variables|Transient|GetCurrentLCDMinutes) (LocalVariables|GetLCDMinutes)) (!= (Variables|Transient|GetCurrentLCDSeconds) (LocalVariables|GetLCDSeconds)))
            (Variables|Transient|SetCurrentLCDMinutes (LocalVariables|GetLCDMinutes))
            (Variables|Transient|SetCurrentLCDSeconds (LocalVariables|GetLCDSeconds))
            (Utilities|FlowControl|Sequence
              (:then_0
                ;[0 1 0] Update the state of the LCD
                (Utilities|FlowControl|Sequence
                  (:then_0
                    (if (<= (Fortnite|Sync|Time|GetServerWorldTimeSeconds) (Variables|Transient|GetArmedFinishedServerTimeSeconds))
                      (LocalVariables|SetLCDState 3.0)
                      (else
                        (switch Utilities|FlowControl|Switch|SwitchonE_TimerObjective_State (Variables|Transient|GetState)
                          (:NewEnumerator0
                            _)
                          (:NewEnumerator7
                            _)
                          (:NewEnumerator1
                            _)
                          (:NewEnumerator3
                            _)
                          (:NewEnumerator4
                            _)
                          (:NewEnumerator5
                            (LocalVariables|SetLCDState 0.0))
                          (:NewEnumerator10
                            _))
                        (LocalVariables|SetLCDState (select (or (Utilities|Enum|Equal(Enum) (Variables|Transient|GetState) "NewEnumerator4") (Math|Boolean|ANDBoolean (Utilities|Enum|Equal(Enum) (Variables|Transient|GetState) "NewEnumerator3") (== (LocalVariables|GetLCDMinutes) 0) (<= (LocalVariables|GetLCDSeconds) 10))) 2.0 1.0)))))
                  (:then_1
                    (Rendering|Material|SetCustomPrimitiveDataFloat (Variables|Debug|GetStaticMeshComponent) (Variables|Material|GetTimerStateCustomPrimitiveIndex) 0.0))))
              (:then_1
                ;[0 0 1] Update the time on the LCD
                (switch Utilities|FlowControl|Switch|SwitchonE_TimerObjective_State (Variables|Transient|GetState)
                  (:NewEnumerator0
                    _)
                  (:NewEnumerator7
                    _)
                  (:NewEnumerator1
                    _)
                  (:NewEnumerator3
                    _)
                  (:NewEnumerator4
                    _)
                  (:NewEnumerator5
                    _))
                (Rendering|Material|SetCustomPrimitiveDataFloat (Variables|Debug|GetStaticMeshComponent) (Variables|Material|GetMinutesCustomPrimitiveIndex) (Math|Conversions|ToFloat(Integer) (LocalVariables|GetLCDMinutes)))
                (Rendering|Material|SetCustomPrimitiveDataFloat (Variables|Debug|GetStaticMeshComponent) (Variables|Material|GetSecondsCustomPrimitiveIndex) (Math|Conversions|ToFloat(Integer) (LocalVariables|GetLCDSeconds))))
              (:then_2
                ;[1 1 0] Update the color of the LCD
                (Rendering|Material|SetCustomPrimitiveDataVector4 (Variables|Debug|GetStaticMeshComponent) (Variables|Material|GetColorCustomPrimitiveIndex) (Math|Vector4|MakeVector4 _returnvalue_r _returnvalue_g _returnvalue_b _returnvalue_a)))
              (:then_3
                ;[1 0 1] Play sound & VFX
                (if (Utilities|Time|DoesTimerExistbyHandle (Variables|Transient|GetTimerHandle))
                  (Utilities|FlowControl|Sequence
                    (:then_0
                      (if (Math|Conversions|ToBoolean(Integer) (Variables|ToyProperties|GetAllowUrgencyMode))
                        (if (and (== (LocalVariables|GetLCDMinutes) 0) (<= (LocalVariables|GetLCDSeconds) (Variables|ToyProperties|GetUrgentModeStartTime)))
                          (CallFunction|TriggerUrgentVFX)
                          (if (Math|Conversions|ToBoolean(Integer) (Variables|ToyProperties|GetPlayTimerSound))
                            (CallFunction|PlaySound (Variables|Audio|GetClockTickUrgentSound) _returnvalue))
                          (else
                            _))
                        (else
                          _))
                      (if (Math|Conversions|ToBoolean(Integer) (Variables|ToyProperties|GetPlayTimerSound))
                        (CallFunction|PlaySound (Variables|Audio|GetClockTickSound) _returnvalue)))
                    (:then_1
                      (if (Math|Boolean|ANDBoolean (Math|Conversions|ToBoolean(Integer) (Variables|ToyProperties|GetAllowUrgencyMode)) (== (LocalVariables|GetLCDMinutes) 0) (== (LocalVariables|GetLCDSeconds) (Variables|ToyProperties|GetUrgentModeStartTime)))
                        (if (Math|Conversions|ToBoolean(Integer) (Variables|ToyProperties|GetSounds))
                          (if (not (Audio|Components|Audio|IsPlaying (Variables|Default|GetUrgent%20Build%20Audio)))
                            (Audio|Components|Audio|Play (Variables|Default|GetUrgent%20Build%20Audio) 0.0)))))))))))))))

(fn UpdateMeshesCollisionProfiles
  (if (Visibility|ShouldbeVisible (Variables|Default|GetStaticMeshVisibleInGameComponent))
    (Collision|SetCollisionProfileName (Variables|Debug|GetStaticMeshComponent) (select (CallFunction|IsHologram) "FortPickupTraceChannelOnly" "FortBuildingMeshPhysics") true :!class /Script/Engine.PrimitiveComponent)
    (else
      (Collision|SetCollisionEnabled (Variables|Debug|GetStaticMeshComponent) "NoCollision"))))

(fn UpdateMinimapTime
  (bind (_minutes _seconds) (Utilities|GetTimeasMinutesSeconds))
  (Utilities|ShouldUpdateClockonServer
    (:Update
      (LocalVariables|SetMinutes _minutes)
      (LocalVariables|SetSeconds _seconds)
      (LocalVariables|SetMinutesAsText (Utilities|Text|ToText(Integer) (LocalVariables|GetMinutes)))
      (LocalVariables|SetSecondsAsText (Utilities|Text|ToText(Integer) (LocalVariables|GetSeconds)))
      (Utilities|FlowControl|Sequence
        (:then_0
          (if (<= (LocalVariables|GetMinutes) 9)
            (LocalVariables|SetMinutesAsText (Utilities|Text|FormatText "NSLOCTEXT(\"[77244AD94C2D2BDBB1DBFDA5068C313D]\", \"5546BA584AFE3261695C16B0937D9E78\", \"0{Minutes}\")" (LocalVariables|GetMinutesAsText)))))
        (:then_1
          (if (<= (LocalVariables|GetSeconds) 9)
            (LocalVariables|SetSecondsAsText (Utilities|Text|FormatText "NSLOCTEXT(\"[77244AD94C2D2BDBB1DBFDA5068C313D]\", \"84593FD3448441024E7623A09EC52169\", \"0{Seconds}\")" (LocalVariables|GetSecondsAsText)))))
        (:then_2
          (Variables|Transient|SetMinimapTime (Utilities|Text|FormatText "NSLOCTEXT(\"[77244AD94C2D2BDBB1DBFDA5068C313D]\", \"1F286A3A427433D287D1F5BD45B1EC8A\", \"{Minutes}:{Seconds}\")" (LocalVariables|GetMinutesAsText) (LocalVariables|GetSecondsAsText))))))
    (:Don't%20Update
      (Variables|Transient|SetMinimapTime (Utilities|Text|MakeLiteralText "")))))

(fn UpdateVisualDisplays
  (Networking|CanExecuteCosmeticEvents
    (:True
      (Utilities|FlowControl|Sequence
        (:then_0
          (CallFunction|UpdateLCD))
        (:then_1
          (if (Math|Conversions|ToBoolean(Integer) (Variables|ToyProperties|GetIsVisibleOnHUD))
            (CallFunction|UpdateHUDData)))
        (:then_2
          (for _index (range (Rendering|Material|GetNumMaterials (Variables|Debug|GetStaticMeshComponent)))
            (Rendering|Material|SetMaterial (Variables|Debug|GetStaticMeshComponent) _index (CallFunction|GetOverrideMeshMaterial (Variables|Debug|GetStaticMeshComponent) _index))))))))

(fn ConstructionScript
  (bind _additional_mesh (Variables|Default|GetAdditional%20Mesh))
  (switch Utilities|FlowControl|Switch|SwitchonE_TimerObjective_MeshOptions (Variables|ToyProperties|GetMeshOptions)
    (:NewEnumerator0
      (Rendering|SetVisibility _additional_mesh false false))
    (:NewEnumerator1
      (Rendering|SetVisibility _additional_mesh true false))))
