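# LCD screen material for the Creative Timer Objective device, showing an emissive minutes-and-seconds readout with a blinking separator.
# Custom primitive data drives Mins, Seconds, Color and TimerState; Fmod splits the time into digits, which offset number-atlas UVs in four SphereMask-bounded slots.
# Higher TimerState values first pulse digit brightness, then swap to a flashing Armed graphic in ArmedColor.
# An LCD pattern function and soft background glow shape the emissive before MF_CreativeMoveTool wraps the attributes.
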Material /Game/Creative/Devices/TimerObjective/Materials/M_TimerObjective_LCD_02
graph {
 properties {bUseMaterialAttributes=true, pixelDepthOffsetMode=PDOM_Legacy, usedShadingModels=Default Lit}
 nodes [
  n1 = VectorParameter() {bUseCustomPrimitiveData=true, parameterName=Color, primitiveDataIndex=3, defaultValue=<rgba 0.0618 0.3305 0.6979 1>}
  n2 = MaterialFunctionCall!(
   U Repeat=ScalarParameter() {defaultValue=35, parameterName=LCD_U}
   V Repeat=ScalarParameter() {defaultValue=15, parameterName=LCD_V}
   Dark Value
   Light Value
  ) {materialFunction=/Game/Effects/Fort_Effects/Effects/Atlas/Materials/MF_Atlas_LCD_01}
  n3 = LinearInterpolate(B=@n1.RGB, Alpha=@n2.Result)
  n4 = Add(A=Multiply(A=TextureCoordinate(), B=Constant2Vector() {g=1, r=0.4}), B=Constant2Vector() {r=-0.13})
  n5 = Sine(None=Time())
  n6 = ScalarParameter() {defaultValue=1.1, parameterName=Uscale}
  n7 = ScalarParameter() {defaultValue=1.1, parameterName=Vscale}
  n8 = Add(
   A=Multiply(A=TextureCoordinate() {uTiling=4}, B=AppendVector(A=@n6, B=@n7))
   B=AppendVector(A=ScalarParameter() {defaultValue=-0.25, parameterName=UBias}, B=ScalarParameter() {defaultValue=-0.05, parameterName=VBias})
  )
  n9 = Multiply(A=@n8, B=Constant2Vector() {g=1, r=0.1})
  n10 = ScalarParameter() {bUseCustomPrimitiveData=true, defaultValue=88, parameterName=Mins, primitiveDataIndex=2}
  n11 = Fmod(A=ScalarParameter() {bUseCustomPrimitiveData=true, defaultValue=88, parameterName=Mins, primitiveDataIndex=2}, B=Constant() {r=10})
  n12 = AppendVector(A=Divide(A=Subtract(A=@n10, B=@n11)) {constB=100}, B=Constant())
  n13 = Multiply(
   A=TextureSample(UVs=Add(A=@n9, B=@n12)) {samplerType=SAMPLERTYPE_Masks, texture=/Game/Creative/Devices/Trigger/Materials/T_NumberSet_Large_CreativeTrigger}.R
   B=SphereMask(
    A=MaterialFunctionCall(Float2=@n8) {materialFunction=&BreakOutFloat2Components}.R
    B=Constant() {r=0.5}
    Radius=Constant() {r=0.5}
    Hardness=Constant() {r=1}
   )
  )
  n14 = Multiply(A=Add(A=@n8, B=Constant2Vector() {r=-1}) {constB=0}, B=Constant2Vector() {g=1, r=0.1})
  n15 = AppendVector(A=Divide(A=@n11) {constB=10}, B=Constant())
  n16 = Multiply(
   A=TextureSample(UVs=Add(A=@n14, B=@n15)) {samplerType=SAMPLERTYPE_Masks, texture=&T_NumberSet_Large_CreativeTrigger}.R
   B=SphereMask(
    A=MaterialFunctionCall(Float2=@n8) {materialFunction=&BreakOutFloat2Components}.R
    B=Constant() {r=1.5}
    Radius=Constant() {r=0.5}
    Hardness=Constant() {r=1}
   )
  )
  n17 = Multiply(A=Add(A=@n8, B=Constant2Vector() {r=-2}) {constB=0}, B=Constant2Vector() {g=1, r=0.1})
  n18 = ScalarParameter() {bUseCustomPrimitiveData=true, defaultValue=88, parameterName=Seconds, primitiveDataIndex=1}
  n19 = Fmod(A=ScalarParameter() {bUseCustomPrimitiveData=true, defaultValue=88, parameterName=Seconds, primitiveDataIndex=1}, B=Constant() {r=10})
  n20 = AppendVector(A=Divide(A=Subtract(A=@n18, B=@n19)) {constB=100}, B=Constant())
  n21 = Multiply(
   A=TextureSample(UVs=Add(A=@n17, B=@n20)) {samplerType=SAMPLERTYPE_Masks, texture=&T_NumberSet_Large_CreativeTrigger}.R
   B=SphereMask(
    A=MaterialFunctionCall(Float2=@n8) {materialFunction=&BreakOutFloat2Components}.R
    B=Constant() {r=2.5}
    Radius=Constant() {r=0.5}
    Hardness=Constant() {r=1}
   )
  )
  n22 = Multiply(A=Add(A=@n8, B=Constant2Vector() {r=-3}) {constB=0}, B=Constant2Vector() {g=1, r=0.1})
  n23 = AppendVector(A=Divide(A=@n19) {constB=10}, B=Constant())
  n24 = Multiply(
   A=TextureSample(UVs=Add(A=@n22, B=@n23)) {samplerType=SAMPLERTYPE_Masks, texture=&T_NumberSet_Large_CreativeTrigger}.R
   B=SphereMask(
    A=MaterialFunctionCall(Float2=@n8) {materialFunction=&BreakOutFloat2Components}.R
    B=Constant() {r=3.5}
    Radius=Constant() {r=0.5}
    Hardness=Constant() {r=1}
   )
  )
  n25 = Add(
   A=Multiply(
    A=TextureSample(UVs=@n4) {samplerType=SAMPLERTYPE_Masks, texture=&/T_NumberSet_Large_CreativeTrigger}.B
    B=Ceil(None=@n5)
   )
   B=Add(A=Add(A=@n13, B=@n16), B=Add(A=@n21, B=@n24))
  )
  n26 = Time()
  n27 = LinearInterpolate(Alpha=ConstantBiasScale(None=Sine(None=@n26))) {constA=1, constB=10}
  n28 = Subtract(A=ScalarParameter() {bUseCustomPrimitiveData=true, parameterName=TimerState}, B=Constant() {r=1})
  n29 = VectorParameter() {parameterName=ArmedColor, defaultValue=<rgba 0.2008 0.8469 0.1015 1>}
  n30 = OneMinus(None=TextureSample(UVs=TextureCoordinate()) {texture=&/T_Armed_TimerObjective}.R)
  n31 = Time()
  n32 = ScalarParameter() {bUseCustomPrimitiveData=true, parameterName=TimerState}
  n33 = Constant() {r=2}
  n34 = LinearInterpolate(
   A=LinearInterpolate(A=@n25, B=Multiply(A=@n25, B=@n27), Alpha=Clamp(None=@n28))
   B=LinearInterpolate(A=Multiply(A=@n29.RGB, B=@n30), B=Constant3Vector(), Alpha=Sine(None=@n31) {period=0.5})
   Alpha=Clamp(None=Subtract(A=@n32, B=@n33))
  )
  n35 = TextureCoordinate()
  n36 = Constant() {r=0.5}
  n37 = ScalarParameter() {defaultValue=0.7, parameterName=BGRadius}
  n38 = Constant()
  n39 = Add(
   A=LinearInterpolate(A=@n34, B=Multiply(A=@n34) {constB=1.3}, Alpha=Power(Base=@n2.Result) {constExponent=6})
   B=Multiply(A=SphereMask(A=@n35, B=@n36, Radius=@n37, Hardness=@n38), B=ScalarParameter() {defaultValue=0.04, parameterName=BGScale})
  )
  n40 = LinearInterpolate(B=@n39, Alpha=@n2.Result)
  n41 = Clamp(None=Subtract(A=ScalarParameter() {bUseCustomPrimitiveData=true, parameterName=TimerState}, B=Constant() {r=2}))
  n42 = SetMaterialAttributes!(
   MaterialAttributes
   Base Color=VectorParameter() {parameterName=BaseColor, defaultValue=<rgba 0.0011 0.0109 0.0434 0>}.RGB
   Metallic=ScalarParameter() {parameterName=Metallic}
   Specular=ScalarParameter() {defaultValue=0.8, parameterName=Specular}
   Roughness=ScalarParameter() {defaultValue=0.1, parameterName=Roughness}
   Emissive Color=Multiply(A=LinearInterpolate(A=@n3, B=@n40, Alpha=@n41), B=@n39)
  ) {attributeSetTypes [
    #1
    #2
    #3
    #4
    #5
   ]}
  MaterialFunctionCall(MA In=@n42) {materialFunction=&MF_CreativeMoveTool}
 ]
}
